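# Unlit material for the Landmass erosion brush that adds procedural noise to a landscape height render target, passing its G channel through.
# A custom HLSL loop sums tiling gradient-noise octaves with optional turbulence, using MPC_Erosion resolution and grid size for repeats and the seed as third coordinate.
# The noise is floored, renormalized and height-scaled.
# An optional min/max slope mask, derived from the normal render target, limits where it applies.
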Material /Landmass/Landscape/ErosionBrush/Materials/AddNoise/ErosionBrush_AddNoise
graph {
 properties {bAllowNegativeEmissiveColor=true, pixelDepthOffsetMode=PDOM_Legacy, shadingModel=MSM_Unlit, usedShadingModels=Unlit}
 nodes [
  Constant()
  n1 = TextureSampleParameter2D!(UVs, Level=Constant(), Apply View MipBias) {mipValueMode=TMVM_MipLevel, parameterName=HeightRT, texture=&Black}
  n2 = CollectionParameter() {collection=/Landmass/Landscape/BlueprintBrushes/MPC/MPC_Erosion, parameterName=ResX}
  n3 = CollectionParameter() {collection=&MPC_Erosion, parameterName=ResY}
  n4 = Divide(
   A=Multiply(
    A=AppendVector(A=@n2, B=@n3)
    B=CollectionParameter() {collection=&MPC_Erosion, parameterName=GridSize}
   )
   B=Multiply(A=ScalarParameter() {defaultValue=20, parameterName=Noise Scale}) {constB=100}
  )
  MaterialExpressionAppendVector_0 = AppendVector(
   A=Multiply(A=TextureCoordinate(), B=NamedRerouteDeclaration(None=@n4) {name=Noise Repeats, nodeColor=<rgba 0 1 0.9766 1>})
   B=ScalarParameter() {parameterName=Noise Seed}
  )
  MaterialExpressionConstant_3 = Constant() {desc="3 = Gradient, 5 = Voronoi", r=3}
  MaterialExpressionScalarParameter_1 = ScalarParameter() {parameterName=Turbulence}
  MaterialExpressionScalarParameter_2 = ScalarParameter() {defaultValue=8, parameterName=Noise Octaves}
  MaterialExpressionScalarParameter_3 = ScalarParameter() {defaultValue=0.5, parameterName=Noise Persistence}
  MaterialExpressionNamedRerouteUsage_0 = NamedRerouteUsage()
  n5 = Custom!(
   Position=@MaterialExpressionAppendVector_0
   Function=@MaterialExpressionConstant_3
   bTurbulence=@MaterialExpressionScalarParameter_1
   Octaves=@MaterialExpressionScalarParameter_2
   Persistence=@MaterialExpressionScalarParameter_3
   RepeatSize=@MaterialExpressionNamedRerouteUsage_0
  ) {
   code="\r\nfloat Out = 0;\r\nfloat OutScale = 1;\r\n\r\n[loop]\r\nfor(int i = 0; i < Octaves; ++i)\r\n{\r\n\r\n\r\n\tif(bTurbulence)\r\n\t{\r\n\t\tOut += abs(Noise3D_Multiplexer(Function, Position, 4, 1, RepeatSize)) * OutScale;\r\n\t}\r\n\telse\r\n\t{\r\n\t\tOut += Noise3D_Multiplexer(Function, Position, 4, 1, RepeatSize) * OutScale;\r\n\t}\r\n\r\n\tPosition *= 2;\r\n\tRepeatSize *= 2;\r\n\tOutScale *= Persistence;\r\n\r\n}\r\n\r\nif(!bTurbulence)\r\n{\r\n\t// bring -1..1 to 0..1 range\r\n\tOut = Out * 0.5f + 0.5f;\r\n}\r\n\r\n// Out is in 0..1 range\r\nreturn lerp(0, 1, Out);"
   description=CustomNoise
   outputType=CMOT_Float1
   inputs [
    {
     inputName=Position
     input {expression=&self:MaterialExpressionAppendVector_0}
    }
    {
     inputName=Function
     input {expression=&self:MaterialExpressionConstant_3}
    }
    {
     inputName=bTurbulence
     input {expression=&self:MaterialExpressionScalarParameter_1}
    }
    {
     inputName=Octaves
     input {expression=&self:MaterialExpressionScalarParameter_2}
    }
    {
     inputName=Persistence
     input {expression=&self:MaterialExpressionScalarParameter_3}
    }
    {
     inputName=RepeatSize
     input {expression=&self:MaterialExpressionNamedRerouteUsage_0}
    }
   ]
  }
  n6 = ScalarParameter() {parameterName=Noise Floor}
  n7 = Subtract(A=@n5, B=@n6)
  n8 = OneMinus(None=@n6)
  n9 = Multiply(A=Max(A=Divide(A=@n7, B=@n8)) {constB=0}, B=Multiply(A=ScalarParameter() {defaultValue=1, parameterName=Noise Height}) {constB=100})
  n10 = NamedRerouteUsage()
  MaterialFunctionCall(X=Add(A=@n1.R, B=Multiply(A=@n9, B=@n10)), Y=@n1.G) {materialFunction=&MakeFloat3}
  n11 = TextureSampleParameter2D() {
   mipValueMode=TMVM_MipLevel
   parameterName=NormalRT
   samplerType=SAMPLERTYPE_LinearColor
   texture=/Engine/ArtTools/RenderToTexture/Textures/127grey
  }
  n12 = ConstantBiasScale(None=ComponentMask(None=@n11.RGB) {g=true, r=true}) {bias=-0.5, scale=2}
  n13 = ComponentMask(None=Abs(None=DeriveNormalZ(InXY=@n12))) {b=true, desc=Cosine}
  NamedRerouteDeclaration(None=Divide(
    A=MaterialFunctionCall!(Side 1 Length=@n13) {materialFunction=/Engine/Functions/Engine_MaterialFunctions02/Math/pythagoreanTheorum}.Side 2 Length
    B=@n13
   ) {desc=Slope}) {name=Terrain Slope, nodeColor=<rgba 0 1 0.3203 1>}
  n14 = NamedRerouteUsage()
  n15 = ScalarParameter() {defaultValue=0.1, parameterName=MinSlope}
  n16 = Saturate(None=Divide(A=Subtract(A=@n14, B=@n15), B=ScalarParameter() {defaultValue=0.1, parameterName=SlopeFalloff})) {desc=Min Slope Mask}
  n17 = ScalarParameter() {defaultValue=0.1, parameterName=SlopeFalloff}
  n18 = Subtract(A=ScalarParameter() {parameterName=MaxSlope}, B=@n17)
  n19 = NamedRerouteUsage()
  n20 = Saturate(None=Divide(A=Subtract(A=@n18, B=@n19), B=@n17)) {desc=Max Slope Mask}
  NamedRerouteDeclaration(None=LinearInterpolate(B=Min(A=@n16, B=@n20), Alpha=ScalarParameter() {parameterName=EnableSlopeMask}) {constA=1}) {name=Slope Mask, nodeColor=<rgba 0.4062 0 1 1>}
  OneMinus(None=@n13)
 ]
}
